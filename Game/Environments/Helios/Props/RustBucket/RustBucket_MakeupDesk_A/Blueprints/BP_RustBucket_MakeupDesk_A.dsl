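; Blueprint for the RustBucket makeup desk prop A, with a vanity mirror mesh whose frame color is set through a variable.
; The construction script runs the parent's logic, then passes MirrorFrameColor and the mirror component to the SetColoronSecondaryMesh function.
; BeginPlay, overlap and Tick events are empty stubs, so the asset's own graph only does construction-time setup.
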(event EventBeginPlay)

(event Collision|EventActorBeginOverlap
  (params (OtherActor :type Actor%20Object%20Reference)))

(event EventTick
  (params (DeltaSeconds :type Float%20%28single-precision%29)))

(fn ConstructionScript
  (|CallParentFunction)
  (Default|SetColoronSecondaryMesh (Variables|Default|GetMirrorFrameColor) (Variables|Default|GetSM_Bathroom_VanityMirror_A)))
